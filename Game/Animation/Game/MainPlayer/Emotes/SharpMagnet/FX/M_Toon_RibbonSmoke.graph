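# Unlit translucent Niagara ribbon material for the smoke trail effect in the player's SharpMagnet emote.
# Particle color RGB drives emissive through an adjustment function.
# Opacity combines particle alpha, a rotated reflected gradient with exponent and multiplier controls, a biased panning smoke texture, and a linear fade along U.
# The smoke samples scaled UV1, with V scale defaulting to zero so the texture varies only along U.
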Material /Game/Animation/Game/MainPlayer/Emotes/SharpMagnet/FX/M_Toon_RibbonSmoke
graph {
 properties {bUsedWithNiagaraRibbons=true, blendMode=BLEND_Translucent, pixelDepthOffsetMode=PDOM_Legacy, shadingModel=MSM_Unlit, usedShadingModels=Unlit}
 nodes [
  n1 = ParticleColor()
  MaterialFunctionCall(Emissive=@n1.RGB) {materialFunction=&MF_FX_EmissiveAdjustments}
  n2 = MaterialFunctionCall!(UVs, Offset, RotationCenter, RotationAngle=Constant() {r=0.25}) {materialFunction=/FortUILibrary/MaterialFunctions/Gradients/MF_UI_ReflectedGradient}
  n3 = ScalarParameter() {defaultValue=1, parameterName=Exp}
  n4 = Saturate(None=Multiply(A=Power(Base=@n2.Result, Exp=@n3), B=ScalarParameter() {defaultValue=4, parameterName=Mult}))
  n5 = TextureCoordinate() {coordinateIndex=1}
  n6 = AppendVector(A=ScalarParameter() {defaultValue=1, parameterName=Param_2}, B=ScalarParameter() {parameterName=Param_1})
  n7 = TextureSample(UVs=Panner(Coordinate=Multiply(A=@n5, B=@n6)) {constCoordinate=1, speedX=-0.5}) {texture=&T_Inky_Smoke_Tile}
  n8 = TextureCoordinate()
  Multiply(A=Multiply(A=Multiply(A=@n1.A, B=@n4), B=Add(A=@n7.R) {constB=0.4}), B=OneMinus(None=ComponentMask(None=@n8) {r=true}))
 ]
}
